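# Niagara vehicle cosmetic for the Skull Chromatic drift trail, drawing RGB-split skulls, ribbons and wheel wraps.
# Skull emitters spawn meshes and sprites at wheel bones, with red and green copies offset sideways and Skull_B defaulting to blue.
# Trail002 streams two backward-moving ribbons with layered color renderers, while burst-spawned wheel-wrap meshes follow the bones.
# Painted-variant overrides and bIsPainted drive colors and alphas; FN_DisableTireFX masks every emitter per wheel.
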NiagaraSystem /VehicleCosmetics/DriftTrail/DriftTrail_SkullChromatic/FX/NS_DriftTrail_SkullChromatic_MP
dependencies {
 usedDataInterfaces [&NiagaraDataInterfaceCurve, &NiagaraDataInterfaceSkeletalMesh]
 usedDynamicInputs []
 usedModules [
  &self:SystemSpawnScript.SystemScriptSource.SystemScriptGraph.NiagaraNodeAssignment_0.SetVariables_#1
  &FN_UnpackVehicleFlags
  /VehicleCosmetics/FX/Niagara/Modules/VC_SetVehicleSpeedAttribs
  &SystemState
  /Game/Effects/Niagara/Modules/Vehicle/FN_VehicleFlags_Cheatcodes
 ]
}
emitters {
 NE_VC_DriftTrail_WheelWrap_Parent {
  emitterUpdateScript [
   &EmitterState(
    Inactive Response=::NewEnumerator1
    Life Cycle Mode=::NewEnumerator1
    Loop Behavior=::NewEnumerator1
    Loop Duration Mode=::NewEnumerator1
    UseLoopDelay=false
   )
   &SpawnBurst_Instantaneous(Spawn Count=2, Use Spawn Probability=false, UseLoopCountLimit=false)
   &SpawnBurst_Instantaneous(Spawn Count=2, Spawn Group=1, Use Spawn Probability=false, UseLoopCountLimit=false) {module=SpawnBurst_Instantaneous001}
   &SpawnBurst_Instantaneous(Spawn Count=2, Spawn Group=2, Use Spawn Probability=false, UseLoopCountLimit=false) {module=SpawnBurst_Instantaneous002}
   &SpawnBurst_Instantaneous(Spawn Count=2, Spawn Group=3, Use Spawn Probability=false, UseLoopCountLimit=false) {module=SpawnBurst_Instantaneous003}
  ]
  particleSpawnScript [
   /SetVariables(Particles.NewVisibilityTag:NiagaraInt32=&SpawnGroup()) {isSetParameters=true}
   &VC_PaintedVariantsOverride(
    Black=<rgba 0.06 0.06 0.06 4>
    BurntSienna=<rgba 0.8 0.1413 0.04 1.5>
    Cobalt=<rgba 0.21 0.4425 3 1.5>
    Crimson=<rgba 0.6458 0.0505 0.0505 1.5>
    DefaultColor=<rgba 1 1 1 4>
    ForestGreen=<rgba 0.2769 1 0.2 1>
    Grey=<rgba 0.2 0.2 0.2 4>
    Lime=<rgba 0.4333 1 0.15 1.5>
    Orange=<rgba 1.25 0.27 0.025 1.5>
    Pink=<rgba 1.25 0.375 0.9583 1>
    Purple=<rgba 0.275 0.125 1.25 0.8>
    R1 Black:LinearColor=<rgba 1 0 0.3333 0.5>
    R1 BurntSienna:LinearColor=<rgba 4 0 0 2>
    R1 Cobalt:LinearColor=<rgba 2 0 1.3333 1>
    R1 Crimson:LinearColor=<rgba 4 0 0 1>
    R1 DefaultColor=<rgba 1 0 0 0.5>
    R1 ForestGreen:LinearColor=<rgba 1 0 0 0.7>
    R1 Gold:LinearColor=<rgba 4 0 0 1>
    R1 Grey:LinearColor=<rgba 0.5 0 0 0.75>
    R1 Lime:LinearColor=<rgba 1 0 0 1>
    R1 Orange:LinearColor=<rgba 2 0 0 1>
    R1 Pink:LinearColor=<rgba 1 0 0.1667 1.5>
    R1 Purple:LinearColor=<rgba 4 0 2.6667 1>
    R1 Saffron:LinearColor=<rgba 2 0.3333 0 2>
    R1 SkyBlue:LinearColor=<rgba 2 0 1 1>
    R1 TitaniumWhite:LinearColor=<rgba 2 0 0 0>
    R2 Black:LinearColor=<rgba 0 1 0 0.75>
    R2 BurntSienna:LinearColor=<rgba 0 1 0 0.5>
    R2 Cobalt:LinearColor=<rgba 0 1 0 1>
    R2 Crimson:LinearColor=<rgba 0.1 1 0.1 0.7>
    R2 DefaultColor:LinearColor=<rgba 0 1 0 0.5>
    R2 ForestGreen:LinearColor=<rgba 0 4 0 1>
    R2 Gold:LinearColor=<rgba 0 4 0 1>
    R2 Grey:LinearColor=<rgba 0 0.5 0 0.75>
    R2 Lime:LinearColor=<rgba 0 4 0 1>
    R2 Orange:LinearColor=<rgba 0.3333 1 0 1>
    R2 Pink:LinearColor=<rgba 0 1 0 0.7>
    R2 Purple:LinearColor=<rgba 0.4 4 0.4 1.5>
    R2 Saffron:LinearColor=<rgba 0.2 4 0.2 2>
    R2 SkyBlue:LinearColor=<rgba 0 4 0 1>
    R2 TitaniumWhite:LinearColor=<rgba 0.04 4 0.04 0>
    R3 Black:LinearColor=<rgba 0 0 1 1>
    R3 BurntSienna:LinearColor=<rgba 0 0 1 0.75>
    R3 Cobalt:LinearColor=<rgba 0 0.0833 1 5>
    R3 Crimson:LinearColor=<rgba 0.3333 0 2 1.25>
    R3 DefaultColor:LinearColor=<rgba 0 0 1 0.5>
    R3 ForestGreen:LinearColor=<rgba 0 0.1667 2 1>
    R3 Gold:LinearColor=<rgba 0 0 1 1>
    R3 Grey:LinearColor=<rgba 0 0 0.5 0.75>
    R3 Lime:LinearColor=<rgba 0 0 1 1>
    R3 Orange:LinearColor=<rgba 0 0 1 1>
    R3 Pink:LinearColor=<rgba 0.1667 0 1 1.5>
    R3 Purple:LinearColor=<rgba 0.125 0 1.5 4>
    R3 Saffron:LinearColor=<rgba 0 0 1 1.5>
    R3 SkyBlue:LinearColor=<rgba 0 0 2 2>
    R3 TitaniumWhite:LinearColor=<rgba 0 0 4 0>
    Saffron=<rgba 1 0.7333 0.2 0.75>
    SkyBlue=<rgba 0.075 0.55 1.5 1>
    TitaniumWhite=<rgba 2 2 2 1>
    Use R1 Painted Overrides=true
    Use R2 Painted Overrides=true
    Use R3 Painted Overrides=true
   )
   &InitializeParticle(
    Color=$Particles.Color
    Mesh Scale Mode=::NewEnumerator4
    Sprite Rotation Mode=::NewEnumerator2
    Sprite Size Mode=::NewEnumerator3
    UsePositionOffset=false
    Write Material Random=true
   )
   &PartitionParticles()
   &SkeletalMeshLocation(
    Bone / Socket Index=&Add_Integer(A=$Particles.PartitionParticles.Partition, B=4)
    Bone / Socket Sampling Mode=::NewEnumerator5
    Interpolate Mesh Sampling=false
    MaskSpawnGroup=false
    Mesh Sampling Type=::NewEnumerator6
    Skeletal Mesh=$System.SkeletalMesh
   )
  ]
  particleUpdateScript [
   &FN_DisableTireFX(
    Mask Index Attribute:NiagaraInt32=$Particles.PartitionParticles.Partition
    Mask Index FL=2
    Mask Index FR=3
    Mask Index RL=0
    Mask Index RR=1
   )
   &VC_WheelWrapScale(Wheel Index Attribute:NiagaraInt32=&Add_Integer(A=$Particles.PartitionParticles.Partition, B=2))
   &ParticleState(Kill Particles When Lifetime Has Elapsed=false)
   &ScaleColor(
    Scale Alpha=&MakeCustomFloatFromBool(Bool=$User.bIsPainted, True Float=0.7, False Float=0.25)
    Scale RGB=&VectorFromFloat(Value=&MakeCustomFloatFromBool(Bool=$User.bIsPainted, True Float=0.7, False Float=0.25))
    ScaleA=true
    ScaleRGB=true
   )
   &UpdateMeshOrientation(
    Rotation Coordinate Space=::NewEnumerator0
    Rotation Vector=<1 0 1>
    Write to Orientation:ENiagara_AttributeSamplingApplyOutput=::NewEnumerator1
   ) {enabled=false}
   &SkeletalMeshLocation(
    Bone / Socket Index=&Add_Integer(A=$Particles.PartitionParticles.Partition, B=4)
    Bone / Socket Sampling Mode=::NewEnumerator5
    MaskSpawnGroup=false
    Mesh Sampling Type=::NewEnumerator6
    Skeletal Mesh=$System.SkeletalMesh
   ) {module=SkeletalMeshLocation001}
   /SetVariables(
    Particles.Blue Ribbon:LinearColor=&VC_DynamicScaleColor(
     Color to Scale:LinearColor=&MakeCustomLinearColorFromBool(Bool=$User.bIsPainted, True Color=$Particles.Initial.R1 Color, False Color=<rgba 0 0 4 1>)
     RGB Scale:Vector3f=&VectorFromFloat(Value=1)
     Alpha Scale:NiagaraFloat=&MakeCustomFloatFromBool(Bool=$User.bIsPainted, True Float=0.4, False Float=0.15)
    )
    Particles.Green Ribbon:LinearColor=&VC_DynamicScaleColor(
     Color to Scale:LinearColor=&MakeCustomLinearColorFromBool(Bool=$User.bIsPainted, True Color=$Particles.Initial.R1 Color, False Color=<rgba 0 4 0 1>)
     RGB Scale:Vector3f=&VectorFromFloat(Value=1)
     Alpha Scale:NiagaraFloat=&MakeCustomFloatFromBool(Bool=$User.bIsPainted, True Float=0.4, False Float=0.1)
    )
    Particles.Red Ribbon:LinearColor=&VC_DynamicScaleColor(
     Color to Scale:LinearColor=&MakeCustomLinearColorFromBool(Bool=$User.bIsPainted, True Color=$Particles.Initial.R1 Color, False Color=<rgba 4 0 0 1>)
     RGB Scale:Vector3f=&VectorFromFloat(Value=1)
     Alpha Scale:NiagaraFloat=&MakeCustomFloatFromBool(Bool=$User.bIsPainted, True Float=0.4, False Float=0.1)
    )
   ) {enabled=false, isSetParameters=true}
   &SolveForcesAndVelocity(Clamp Velocity=false, Limit Acceleration=false)
   /SetVariables(
    Particles.Blue Ribbon:LinearColor=&VC_DynamicScaleColor(
     Color to Scale:LinearColor=$Particles.Initial.R3 Color
     RGB Scale:Vector3f=&VectorFromFloat(Value=1)
     Alpha Scale:NiagaraFloat=&MakeCustomFloatFromBool(Bool=$User.bIsPainted, True Float=0.1, False Float=0.15)
    )
    Particles.Green Ribbon:LinearColor=&VC_DynamicScaleColor(
     Color to Scale:LinearColor=$Particles.Initial.R2 Color
     RGB Scale:Vector3f=&VectorFromFloat(Value=1)
     Alpha Scale:NiagaraFloat=&MakeCustomFloatFromBool(Bool=$User.bIsPainted, True Float=0.1, False Float=0.1)
    )
    Particles.Red Ribbon:LinearColor=&VC_DynamicScaleColor(
     Color to Scale:LinearColor=$Particles.Initial.R1 Color
     RGB Scale:Vector3f=&VectorFromFloat(Value=1)
     Alpha Scale:NiagaraFloat=&MakeCustomFloatFromBool(Bool=$User.bIsPainted, True Float=0.1, False Float=0.1)
    )
   ) {isSetParameters=true}
  ]
  rendererClasses [&NiagaraMeshRendererProperties]
  renderers [
   {rendererClass=&NiagaraMeshRendererProperties, rendererIndex=0}
   {rendererClass=&NiagaraMeshRendererProperties, rendererIndex=1}
   {rendererClass=&NiagaraMeshRendererProperties, rendererIndex=2}
   {rendererClass=&NiagaraMeshRendererProperties, rendererIndex=3}
  ]
 }
 Skull {
  emitterUpdateScript [
   &EmitterState(
    Inactive Response=::NewEnumerator1
    Life Cycle Mode=::NewEnumerator1
    Loop Behavior=::NewEnumerator1
    Loop Duration Mode=::NewEnumerator1
    UseLoopDelay=false
   )
   &SpawnRate(SpawnRate=4, Use Spawn Probability=false)
   &SpawnBurst_Instantaneous(Spawn Count=2, Use Spawn Probability=false, UseLoopCountLimit=false)
  ]
  particleSpawnScript [
   &VC_PaintedVariantsOverride(
    Black=<rgba 0.03 0.03 0.03 1>
    BurntSienna=<rgba 0.1667 0.0339 0.0074 0.6348>
    Cobalt=<rgba 0.0833 0.1003 1 1>
    Crimson=<rgba 0.4 0 0.0108 0.6>
    Lime=<rgba 0.474 0.9479 0 1>
    Orange=<rgba 0.6519 0.1327 0 1>
    Pink=<rgba 0.9167 0.2769 0.684 1>
    Purple=<rgba 0.0067 0 0.08 1>
    R1 Black:LinearColor=<rgba 0.187 0.187 0.2031 1>
    R1 BurntSienna:LinearColor=<rgba 0.1049 0.012 0 1>
    R1 Cobalt:LinearColor=<rgba 0.0239 0.0287 0.2865 1>
    R1 Crimson:LinearColor=<rgba 0.5083 0.0534 0.0548 1>
    R1 ForestGreen:LinearColor=<rgba 0 0.0608 0.002 1>
    R1 Grey:LinearColor=<rgba 0.011 0.011 0.011 1>
    R1 Lime:LinearColor=<rgba 0.2881 0.569 0.116 1>
    R1 Orange:LinearColor=<rgba 0.2265 0.0457 0 1>
    R1 Pink:LinearColor=<rgba 0.5338 0.0533 0.5359 1>
    R1 Purple:LinearColor=<rgba 0.2523 0.1221 0.7127 1>
    R1 Saffron:LinearColor=<rgba 0.6961 0.6646 0.2577 1>
    R1 SkyBlue:LinearColor=<rgba 0.2143 0.7182 0.6957 1>
    R1 TitaniumWhite:LinearColor=<rgba 0.099 0.099 0.099 1>
    Saffron=<rgba 0.474 0.3634 0 1>
    SkyBlue=<rgba 0 0.2536 0.7292 1>
    TitaniumWhite=<rgba 0.6927 0.6927 0.6927 1>
    Use R1 Painted Overrides=true
   )
   &InitializeParticle(
    Color=$Particles.Color
    Mesh Scale Mode=::NewEnumerator3
    Mesh Uniform Scale=2.2
    Sprite Rotation Mode=::NewEnumerator2
    Sprite Size Mode=::NewEnumerator3
    UsePositionOffset=false
    Write Material Random=true
   )
   &PartitionParticles()
   &SkeletalMeshLocation(
    Bone / Socket Index=&Add_Integer(A=$Particles.PartitionParticles.Partition, B=4)
    Bone / Socket Sampling Mode=::NewEnumerator5
    Interpolate Mesh Sampling=false
    MaskSpawnGroup=false
    Mesh Sampling Type=::NewEnumerator6
    Sampled Position Offset=&AddVectorToPosition(Position=$User.offset, Vector=<-85 0 -5>)
    Skeletal Mesh=$System.SkeletalMesh
   )
   &AddVelocity(Velocity=&VC_Mask_Vectors(
     FrontEnd Value:Vector3f=<-1500 0 0>
     blsFrontEndPreview:NiagaraBool=false
     SpeedMin Value:Vector3f=<-800 0 0>
     SpeedMax Value:Vector3f=<-1 0 0>
     Speed Alpha:NiagaraFloat=$System.Speed_Alpha
    )) {enabled=false}
   &InitialMeshOrientation(
    Mesh Orientation Mode=::NewEnumerator7
    Rotation=&RandomRangeVector(Minimum=<0 0.11 0>, Maximum=<0 0.07 0>)
    Use Rotation Vector=true
   )
   /SetVariables(Particles.Scale Skulls:Vector3f=&VectorFromFloat(Value=&RandomRangeFloat(Minimum=2.25, Maximum=2.3))) {isSetParameters=true}
  ]
  particleUpdateScript [
   &FN_DisableTireFX(
    Mask Index Attribute:NiagaraInt32=$Particles.PartitionParticles.Partition
    Mask Index FL=2
    Mask Index FR=3
    Mask Index RL=0
    Mask Index RR=1
   )
   &ParticleState()
   &UpdateMeshOrientation(Rotation Vector=&RandomRangeVector(Minimum=<0 0.05 0.05>, Maximum=<0 -0.05 -0.05>))
   &ScaleColor(
    Scale Alpha=&FloatFromCurve(FloatCurve=DataInterface() {propertyValues {Curve {keys [<at 0 0>,<at 0.2 1>,<at 0.8 1>,<at 1 0>]}}})
    Scale RGB=<0.8 0.8 0.8>
    ScaleA=true
    ScaleRGB=true
   )
   /SetVariables(
    Particles.Skull G Color:LinearColor=&VC_DynamicScaleColor(
     Color to Scale:LinearColor=&MakeCustomLinearColorFromBool(Bool=$User.bIsPainted, True Color=$Particles.R1 Color, False Color=<rgba 0 2 0 1>)
     RGB Scale:Vector3f=<1 1 1>
     Alpha Scale:NiagaraFloat=&FloatFromCurve(FloatCurve=DataInterface() {propertyValues {Curve {keys [<at 0 0>,<at 0.4 1>,<at 1 0>]}}}, Scale Curve=2)
    )
    Particles.Skull R Color:LinearColor=&VC_DynamicScaleColor(
     Color to Scale:LinearColor=&MakeCustomLinearColorFromBool(Bool=$User.bIsPainted, True Color=$Particles.Color, False Color=<rgba 2 0 0 1>)
     RGB Scale:Vector3f=<1 1 1>
     Alpha Scale:NiagaraFloat=&FloatFromCurve(FloatCurve=DataInterface() {propertyValues {Curve {keys [<at 0 0>,<at 0.4 1>,<at 1 0>]}}}, Scale Curve=2)
    )
    Particles.SkullG Pos:NiagaraPosition=&AddVectorToPosition(Position=$Particles.Position, Vector=/VehicleCosmetics/FX/Niagara/Modules/VC_ConvertVectorToWorld(Vector:Vector3f=<0 -1 0>))
    Particles.SkullR Pos:NiagaraPosition=&AddVectorToPosition(Position=$Particles.Position, Vector=&VC_ConvertVectorToWorld(Vector:Vector3f=<0 1 0>))
   ) {isSetParameters=true}
   &SolveForcesAndVelocity(Clamp Velocity=false, Limit Acceleration=false)
  ]
  rendererClasses [&NiagaraMeshRendererProperties, &NiagaraSpriteRendererProperties]
  renderers [
   {rendererClass=&NiagaraMeshRendererProperties, rendererIndex=0}
   {rendererClass=&NiagaraMeshRendererProperties, rendererIndex=1}
   {rendererClass=&NiagaraMeshRendererProperties, rendererIndex=2}
   {rendererClass=&NiagaraSpriteRendererProperties, rendererIndex=3}
  ]
 }
 Skull_B {
  emitterUpdateScript [
   &EmitterState(
    Inactive Response=::NewEnumerator1
    Life Cycle Mode=::NewEnumerator1
    Loop Behavior=::NewEnumerator1
    Loop Delay=0.1
    Loop Duration Mode=::NewEnumerator1
    UseLoopDelay=true
   )
   &SpawnBurst_Instantaneous(Use Spawn Probability=false, UseLoopCountLimit=false)
   &SpawnRate(SpawnRate=4, Use Spawn Probability=false)
  ]
  particleSpawnScript [
   &VC_PaintedVariantsOverride(
    Black=<rgba 0.0469 0.0469 0.0469 1>
    BurntSienna=<rgba 0.1823 0.0625 0.035 1>
    Cobalt=<rgba 0.0454 0.0505 0.3229 1>
    Crimson=<rgba 0.0387 0 0.0001 1>
    DefaultColor=<rgba 0 0 4 1>
    ForestGreen=<rgba 0.0925 0.558 0.11 1>
    Grey=<rgba 0.0276 0.0276 0.0276 1>
    Lime=<rgba 0.0505 0.1326 0 1>
    Orange=<rgba 0.8784 0.22 0.0534 1>
    Pink=<rgba 0.1693 0.0169 0.1713 1>
    Purple=<rgba 0.0276 0 0.125 1>
    Saffron=<rgba 0.0221 0.0207 0 1>
    SkyBlue=<rgba 0 0.0497 0.0476 1>
    TitaniumWhite=<rgba 0.5885 0.5885 0.5885 1>
   )
   &InitializeParticle(
    Color=$Particles.Color
    Mesh Scale Mode=::NewEnumerator3
    Mesh Uniform Scale=&RandomRangeFloat(Minimum=2.25, Maximum=2.3)
    Sprite Rotation Mode=::NewEnumerator2
    Sprite Size Mode=::NewEnumerator3
    UsePositionOffset=false
    Write Material Random=true
   )
   &PartitionParticles()
   &SkeletalMeshLocation(
    Bone / Socket Index=&Add_Integer(A=$Particles.PartitionParticles.Partition, B=4)
    Bone / Socket Sampling Mode=::NewEnumerator5
    Interpolate Mesh Sampling=false
    MaskSpawnGroup=false
    Mesh Sampling Type=::NewEnumerator6
    Sampled Position Offset=&AddVectorToPosition(Position=$User.offset, Vector=<-85 0 -5>)
    Skeletal Mesh=$System.SkeletalMesh
   )
   &AddVelocity(Velocity=&VC_Mask_Vectors(
     FrontEnd Value:Vector3f=<-1500 0 0>
     blsFrontEndPreview:NiagaraBool=false
     SpeedMin Value:Vector3f=<-800 0 0>
     SpeedMax Value:Vector3f=<-1 0 0>
     Speed Alpha:NiagaraFloat=$System.Speed_Alpha
    )) {enabled=false}
   &InitialMeshOrientation(Mesh Orientation Mode=::NewEnumerator7, Rotation=<0 0.1 0>, Use Rotation Vector=true)
  ]
  particleUpdateScript [
   &FN_DisableTireFX(
    Mask Index Attribute:NiagaraInt32=$Particles.PartitionParticles.Partition
    Mask Index FL=2
    Mask Index FR=3
    Mask Index RL=0
    Mask Index RR=1
   )
   &ParticleState()
   &UpdateMeshOrientation(Rotation Vector=&RandomRangeVector(Minimum=<0 0.09 0.09>, Maximum=<0 -0.09 -0.09>))
   &ScaleColor(
    Scale Alpha=&FloatFromCurve(FloatCurve=DataInterface() {propertyValues {Curve {keys [<at 0 0>,<at 0.4 1>,<at 1 0>]}}})
    Scale RGB=<3 3 3>
    ScaleA=true
    ScaleRGB=true
   )
   &SolveForcesAndVelocity(Clamp Velocity=false, Limit Acceleration=false)
  ]
  rendererClasses [&NiagaraMeshRendererProperties, &NiagaraSpriteRendererProperties]
  renderers [
   {rendererClass=&NiagaraMeshRendererProperties, rendererIndex=0}
   {rendererClass=&NiagaraSpriteRendererProperties, rendererIndex=1}
  ]
 }
 Trail002 {
  emitterUpdateScript [
   &EmitterState()
   &VC_RateQualityHandler(Epic Spawn Rate=90, High Spawn Rate=60, Low Spawn Rate=30, Medium Spawn Rate=45, Use Custom Rate?=true) {enabled=false}
   &VC_SpawnRibbon_v1(Number of Ribbons=2, SpawnRate=30)
  ]
  particleSpawnScript [
   &VC_PaintedVariantsOverride(
    Black=<rgba 0.001 0.001 0.001 1>
    BurntSienna=<rgba 0.018 0.0035 0.0018 1>
    Cobalt=<rgba 0.0033 0.0088 0.3304 1>
    Crimson=<rgba 0.025 0.0002 0.0002 1>
    DefaultColor=<rgba 0.1 0.1 0.1 2>
    ForestGreen=<rgba 0.0015 0.02 0 1>
    Grey=<rgba 0.004 0.004 0.004 1>
    Lime=<rgba 0.0233 0.0729 0.0036 1>
    Orange=<rgba 0.04 0.0081 0 1>
    Pink=<rgba 0.1823 0.0159 0.1421 0.5>
    Purple=<rgba 0.0026 0.0001 0.015 1>
    R1 Black:LinearColor=<rgba 4 0 0 1>
    R1 BurntSienna:LinearColor=<rgba 4 0 0 1>
    R1 Cobalt:LinearColor=<rgba 4 0 0 1>
    R1 Crimson:LinearColor=<rgba 4 0 0 1>
    R1 DefaultColor=<rgba 4 0.1333 0 1>
    R1 ForestGreen:LinearColor=<rgba 4 0 0 1>
    R1 Gold:LinearColor=<rgba 4 0 0 1>
    R1 Grey:LinearColor=<rgba 4 0 0 1>
    R1 Lime:LinearColor=<rgba 4 0 0 1>
    R1 Orange:LinearColor=<rgba 4 0 0 1>
    R1 Pink:LinearColor=<rgba 4 0 0 1>
    R1 Purple:LinearColor=<rgba 4 0 0 1>
    R1 Saffron:LinearColor=<rgba 4 0 0 1>
    R1 SkyBlue:LinearColor=<rgba 4 0 0 1>
    R1 TitaniumWhite:LinearColor=<rgba 4 0 0 0>
    R2 Black:LinearColor=<rgba 0 2 0 2>
    R2 BurntSienna:LinearColor=<rgba 0 2 0 2>
    R2 Cobalt:LinearColor=<rgba 0 2 0 2>
    R2 Crimson:LinearColor=<rgba 0 2 0 2>
    R2 DefaultColor:LinearColor=<rgba 0 4 0 2>
    R2 ForestGreen:LinearColor=<rgba 0 2 0 2>
    R2 Gold:LinearColor=<rgba 0 2 0 2>
    R2 Grey:LinearColor=<rgba 0 2 0 2>
    R2 Lime:LinearColor=<rgba 0 2 0 2>
    R2 Orange:LinearColor=<rgba 0 2 0 2>
    R2 Pink:LinearColor=<rgba 0 2 0 2>
    R2 Purple:LinearColor=<rgba 0 2 0 2>
    R2 Saffron:LinearColor=<rgba 0 2 0 2>
    R2 SkyBlue:LinearColor=<rgba 0 2 0 2>
    R2 TitaniumWhite:LinearColor=<rgba 0 2 0 0>
    R3 Black:LinearColor=<rgba 0 0.2 4 5>
    R3 BurntSienna:LinearColor=<rgba 0 0.2 4 5>
    R3 Cobalt:LinearColor=<rgba 0 0.2 4 5>
    R3 Crimson:LinearColor=<rgba 0 0.2 4 5>
    R3 DefaultColor:LinearColor=<rgba 0 0.2 4 5>
    R3 ForestGreen:LinearColor=<rgba 0 0.2 4 5>
    R3 Gold:LinearColor=<rgba 0 0.2 4 5>
    R3 Grey:LinearColor=<rgba 0 0.2 4 5>
    R3 Lime:LinearColor=<rgba 0 0.2 4 5>
    R3 Orange:LinearColor=<rgba 0 0.2 4 5>
    R3 Pink:LinearColor=<rgba 0 0.2 4 5>
    R3 Purple:LinearColor=<rgba 0 0.2 4 5>
    R3 Saffron:LinearColor=<rgba 0 0.2 4 5>
    R3 SkyBlue:LinearColor=<rgba 0 0.2 4 5>
    R3 TitaniumWhite:LinearColor=<rgba 0 0.2 4 0>
    R4 Black:LinearColor=<rgba 0.02 0.02 0.02 1.5>
    R4 BurntSienna:LinearColor=<rgba 0.1 0.0205 0.01 1.5>
    R4 Cobalt:LinearColor=<rgba 0.024 0.0632 1.2 2>
    R4 Crimson:LinearColor=<rgba 0.15 0.009 0.009 1>
    R4 ForestGreen:LinearColor=<rgba 0.0195 0.15 0.0195 2>
    R4 Grey:LinearColor=<rgba 0.06 0.06 0.06 2>
    R4 Lime:LinearColor=<rgba 0.0747 0.2 0.012 1>
    R4 Orange:LinearColor=<rgba 0.3 0.0525 0.003 1>
    R4 Pink:LinearColor=<rgba 0.95 0.0712 0.8035 1>
    R4 Purple:LinearColor=<rgba 0.0339 0.012 0.2 1.5>
    R4 Saffron:LinearColor=<rgba 0.2 0.1717 0.03 1.5>
    R4 SkyBlue:LinearColor=<rgba 0.024 0.1493 0.4 1>
    R4 TitaniumWhite:LinearColor=<rgba 0.4 0.4 0.4 0.15>
    Saffron=<rgba 0.0665 0.0729 0.0034 1>
    SkyBlue=<rgba 0.0024 0.0186 0.0573 0.6174>
    TitaniumWhite=<rgba 0.2135 0.2135 0.2135 1>
    Use R1 Painted Overrides=true
    Use R2 Painted Overrides=true
    Use R3 Painted Overrides=true
    Use R4 Painted Overrides=true
   )
   &VC_InitializeRibbon_v1(
    Color=$Particles.Color
    Custom Offset=true
    Enable Skeletal Position=true
    Lifetime=0.55
    Local Vector Offset=&AddVectorToPosition(Position=$User.offset, Vector=<-75 0 -3>)
    Offset Amount=1
    Ribbon Width:NiagaraFloat=&MakeCustomFloatFromBool(Bool=$User.bIsPainted, True Float=50, False Float=45)
   )
   &AddVelocity(Velocity=<-2000 0 0>) {module=AddVelocity001}
  ]
  particleUpdateScript [
   &FN_DisableTireFX(Mask Index Attribute:NiagaraInt32=$Particles.SpawnGroup, Mask Index FL=2, Mask Index FR=3, Mask Index RL=0, Mask Index RR=1)
   &ParticleState()
   &InheritSourceMovement(
    Applied Position Scale=&Multiply_VectorByFloat(
     Vector=<1 1 1>
     Float=&FloatFromCurve(FloatCurve=DataInterface() {propertyValues {Curve {keys [<at 0.3 1 RCIM_Cubic>, {arriveTangent=-3.0131, interpMode=RCIM_Cubic, leaveTangent=-3.0131, tangentMode=RCTM_User, time=1, value=0}]}}})
    )
    ApplyForce=false
    ApplyPosition=true
    ApplyPositionFalloff=false
   )
   &SolveForcesAndVelocity(Clamp Velocity=false, Limit Acceleration=false)
   &ScaleColor(
    Scale Alpha=&MakeCustomFloatFromBool(Bool=$User.bIsPainted, True Float=0, False Float=5)
    Scale RGB=<60 60 60>
    ScaleA=true
    ScaleRGB=true
   )
   &DynamicMaterialParameters(
    Index 0 Param 1=&FloatFromCurve(FloatCurve=DataInterface() {propertyValues {Curve {keys [<at 0.5 1>,<at 1 5>]}}}, Scale Curve=0.9)
    Index 0 Param 2=1
    Index 0 Param 3=&MaskFloatBySpawnGroup(Masked Float=0, Passthrough Float=0.5, Spawn Group Mask=&SpawnGroup())
    Index 0 Param 4=1
    Index 1 Param 1=&FloatFromCurve(FloatCurve=DataInterface() {propertyValues {Curve {keys [<at 0.5 0.7>,<at 1 1.5>]}}})
    Index 1 Param 2=1
    Index 1 Param 3=&MaskFloatBySpawnGroup(Masked Float=0, Passthrough Float=0.5, Spawn Group Mask=&SpawnGroup())
    Index 1 Param 4=1
    Module.Index0Float0Write=true
    Module.Index0Float1Write=true
    Module.Index0Float2Write=true
    Module.Index0Float3Write=true
    Module.Index1Float0Write=true
    Module.Index1Float1Write=true
    Module.Index1Float2Write=true
    Module.Index1Float3Write=true
    Write Parameter Index 1=::NewEnumerator1
   )
   &ScaleRibbonWidth(Ribbon Width Scale=&FloatFromCurve(FloatCurve=DataInterface() {propertyValues {Curve {keys [<at 0 0.5>,<at 0.1 0.8>,<at 0.5 1>,<at 1.5 0>]}}}))
   /SetVariables(
    Particles.Blue Ribbon:LinearColor=&VC_DynamicScaleColor(
     Color to Scale:LinearColor=&MakeCustomLinearColorFromBool(Bool=$User.bIsPainted, True Color=$Particles.R3 Color, False Color=<rgba 0 0.0667 4 1>)
     RGB Scale:Vector3f=<1 1 1>
     Alpha Scale:NiagaraFloat=&MakeCustomFloatFromBool(
      Bool=$User.bIsPainted
      True Float=&FloatFromCurve(FloatCurve=DataInterface() {propertyValues {Curve {keys [<at 0.04 0 RCIM_Cubic>,<at 0.2 1.25>,<at 1 1>]}}}, Scale Curve=1.1)
      False Float=1.1
     )
    )
    Particles.Green Ribbon:LinearColor=&VC_DynamicScaleColor(
     Color to Scale:LinearColor=$Particles.R2 Color
     RGB Scale:Vector3f=<1 1 1>
     Alpha Scale:NiagaraFloat=&MakeCustomFloatFromBool(
      Bool=$User.bIsPainted
      True Float=&FloatFromCurve(
       FloatCurve=DataInterface() {propertyValues {Curve {
          defaultValue=1
          keys [{interpMode=RCIM_Cubic, tangentMode=RCTM_User, time=0, value=0.5}, {interpMode=RCIM_Cubic, tangentMode=RCTM_User, time=0.5, value=1}]
         }}}
       Scale Curve=0.8
      )
      False Float=&FloatFromCurve(
       FloatCurve=DataInterface() {propertyValues {Curve {defaultValue=1, keys [{tangentMode=RCTM_User, time=0, value=0.5}, {tangentMode=RCTM_User, time=1, value=1}]}}}
       Scale Curve=0.8
      )
     )
    )
    Particles.Red Ribbon:LinearColor=&VC_DynamicScaleColor(
     Color to Scale:LinearColor=$Particles.R1 Color
     RGB Scale:Vector3f=<1 1 1>
     Alpha Scale:NiagaraFloat=&MakeCustomFloatFromBool(
      Bool=$User.bIsPainted
      True Float=&FloatFromCurve(FloatCurve=DataInterface() {propertyValues {Curve {keys [<at 0 1 RCIM_Cubic>, {arriveTangent=-0.5, interpMode=RCIM_Cubic, leaveTangent=-0.5, time=0.25, value=0.5}, <at 1 0.5 RCIM_Cubic>]}}})
      False Float=1
     )
    )
    Particles.SCALEPAINTEDV:LinearColor=&VC_DynamicScaleColor(
     Color to Scale:LinearColor=$Particles.R4 Color
     RGB Scale:Vector3f=&VectorFromFloat(Value=&FloatFromCurve(FloatCurve=DataInterface() {propertyValues {Curve {defaultValue=1, keys [<at 0 0.5 RCIM_Cubic>,<at 0.5 1.5 RCIM_Cubic>]}}}, Scale Curve=5))
     Alpha Scale:NiagaraFloat=&MakeCustomFloatFromBool(
      Bool=$User.bIsPainted
      True Float=&FloatFromCurve(
       FloatCurve=DataInterface() {propertyValues {Curve {keys [
           <at 0 0 RCIM_Cubic>
           {arriveTangent=0.75, interpMode=RCIM_Cubic, leaveTangent=0.75, time=0.1, value=0.9}
           {arriveTangent=0.1667, interpMode=RCIM_Cubic, leaveTangent=0.1667, time=0.4, value=0.3}
           <at 0.7 1 RCIM_Cubic>
          ]}}}
       Scale Curve=30
      )
      False Float=0
     )
    )
   ) {isSetParameters=true}
   &VC_UpdateRibbon_v1(Enable Skeletal Position=true)
  ]
  rendererClasses [&NiagaraRibbonRendererProperties]
  renderers [
   {rendererClass=&NiagaraRibbonRendererProperties}
   {rendererClass=&NiagaraRibbonRendererProperties, rendererIndex=1}
   {rendererClass=&NiagaraRibbonRendererProperties, rendererIndex=2}
   {rendererClass=&NiagaraRibbonRendererProperties, rendererIndex=3}
   {rendererClass=&NiagaraRibbonRendererProperties, rendererIndex=4}
  ]
 }
}
summary {
 systemName=NS_DriftTrail_SkullChromatic_MP
 emitters [
  {emitterName=Skull, rendererClasses [&NiagaraMeshRendererProperties, &NiagaraSpriteRendererProperties]}
  {emitterName=Skull_B, rendererClasses [&NiagaraMeshRendererProperties, &NiagaraSpriteRendererProperties]}
  {emitterName=NE_VC_DriftTrail_WheelWrap_Parent, rendererClasses [&NiagaraMeshRendererProperties]}
  {emitterName=Trail002, rendererClasses [&NiagaraRibbonRendererProperties]}
 ]
 userVariables [
  User.bIsFrontendPreview {type=NiagaraBool}
  User.bIsPainted {type=NiagaraBool}
  User.ColorVariantID {type=NiagaraInt32}
  User.offset {type=NiagaraPosition, description="", defaultValue {}}
  User.PaintedColor {type=LinearColor, description="", defaultValue {struct=LinearColor, value=<rgba 1 1 1 1>}}
  User.PaintedScalar {type=NiagaraFloat, defaultValue=1, description=""}
  User.VehicleFlags {type=NiagaraInt32, defaultValue=806289408}
  User.Wheel_Radius_Array {
   type=NiagaraDataInterfaceArrayFloat
   defaultValue {
    struct=NiagaraExt_VariableValue_DataInterface
    value {
     dataInterface=&self:NiagaraDataInterfaceArrayFloat_1
     dataInterfaceClass=&NiagaraDataInterfaceArrayFloat
    }
   }
  }
  User.Wheel_Width_Array {
   type=NiagaraDataInterfaceArrayFloat
   defaultValue {
    struct=NiagaraExt_VariableValue_DataInterface
    value {
     dataInterface=&self:NiagaraDataInterfaceArrayFloat_0
     dataInterfaceClass=&NiagaraDataInterfaceArrayFloat
    }
   }
  }
  User.WheelTouchingArray {
   type=NiagaraDataInterfaceArrayFloat
   defaultValue {
    struct=NiagaraExt_VariableValue_DataInterface
    value {
     dataInterface=&self:NiagaraDataInterfaceArrayFloat_2
     dataInterfaceClass=&NiagaraDataInterfaceArrayFloat
    }
   }
  }
 ]
}
