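# Translucent lit material for Niagara mesh-particle skulls in the SkullChromatic drift trail cosmetic.
# A skull texture, particle color, panning cloud noise and Fresnel terms build the look, with static switches selecting RGB-tinted variants through an emissive-adjustment function.
# Procedural dot-product circles in UV space form an eye mask.
# A world position offset chain combines layered sine/cosine drift with an oscillating rotation about the pivot, scaled by green vertex color.
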Material /VehicleCosmetics/DriftTrail/DriftTrail_SkullChromatic/Materials/M_DriftTrail_SkullChromaticRGB
graph {
 properties {bUsedWithNanite=true, bUsedWithNiagaraMeshParticles=true, blendMode=BLEND_Translucent, usedShadingModels=Default Lit}
 nodes [
  n1 = NamedRerouteUsage()
  n2 = MaterialFunctionCall(Invert Fresnel=StaticBool() {value=true}, Power=Constant() {r=1.5}) {materialFunction=&Fresnel_Function}
  n3 = Power(Base=@n2.Result) {constExponent=1.5}
  n4 = ParticleColor()
  n5 = Multiply(
   A=TextureSample() {texture=/DealershipSharedLibrary/FX/Textures/Diffuse/T_VCVFX_Hat_Skullfire_D}.R
   B=ScalarParameter() {defaultValue=3, parameterName=SkullPaintedTextureLightness}
  )
  MaterialExpressionConstantClamp_1 = Saturate(None=@n5)
  n6 = StaticSwitchParameter(
   True=Multiply(A=@n4.RGB, B=@MaterialExpressionConstantClamp_1)
   False=Power(Base=@MaterialExpressionConstantClamp_1, Exp=ScalarParameter() {defaultValue=0.4, parameterName=SkullOpacity})
  ) {parameterName=RGB?}
  n7 = TextureSample(UVs=Panner()) {samplerType=SAMPLERTYPE_LinearColor, texture=/DealershipSharedLibrary/FX/Textures/Noise/T_VCVFX_NoiseClouds}
  n8 = Multiply(A=Multiply(A=Power(Base=@n7.RGB) {constExponent=3}, B=Constant() {r=0.5}), B=NamedRerouteUsage())
  n9 = NamedRerouteUsage()
  n10 = Multiply(A=@n4.RGB, B=Multiply(A=@n6, B=Add(A=@n8, B=@n9)))
  MaterialExpressionConstantClamp_0 = Saturate(None=@n10)
  MaterialFunctionCall(Emissive=StaticSwitchParameter(True=Multiply(A=@n1, B=@n3), False=@MaterialExpressionConstantClamp_0) {parameterName=IsRGBSkulls?}) {materialFunction=&MF_FX_EmissiveAdjustments}
  n11 = Sine(None=MaterialFunctionCall(Speed=ScalarParameter() {defaultValue=1.5, parameterName=SpeedDB}) {materialFunction=&TimeWithSpeedVariable}.Time)
  n12 = Sine(None=MaterialFunctionCall(Speed=ScalarParameter() {defaultValue=1, parameterName=SpeedAB}) {materialFunction=&TimeWithSpeedVariable}.Time)
  n13 = Sine(None=MaterialFunctionCall(Speed=ScalarParameter() {defaultValue=-1.5, parameterName=SpeedD}) {materialFunction=&TimeWithSpeedVariable}.Time)
  n14 = Sine(None=MaterialFunctionCall(Speed=ScalarParameter() {defaultValue=-1, parameterName=SpeedA}) {materialFunction=&TimeWithSpeedVariable}.Time)
  n15 = Sine(None=MaterialFunctionCall(Speed=ScalarParameter() {defaultValue=1.5, parameterName=SpeedC}) {materialFunction=&TimeWithSpeedVariable}.Time)
  n16 = Time()
  n17 = ScalarParameter() {defaultValue=1, parameterName=TimeOffset}
  n18 = Cosine(None=Multiply(A=ScalarParameter() {defaultValue=1, parameterName=SpeedB}, B=Add(A=@n16, B=@n17)))
  n19 = MaterialFunctionCall(
   X=Multiply(A=LinearInterpolate(B=@n11, Alpha=@n12) {constA=1}, B=ScalarParameter() {defaultValue=1, parameterName=X_Drift_Mult})
   Y=Multiply(A=LinearInterpolate(B=@n13, Alpha=@n14) {constA=1}, B=ScalarParameter() {defaultValue=1, parameterName=Y_Drift_Mult})
   Z=Multiply(A=LinearInterpolate(A=@n15, Alpha=@n18) {constB=-1}, B=ScalarParameter() {defaultValue=1, parameterName=Z_Drift_Mult})
  ) {materialFunction=&MakeFloat3}
  n20 = ScalarParameter() {defaultValue=0.5, parameterName=Uniform_Drift_Mult}
  n21 = Constant3Vector() {constant=<rgba 0 -1 0 0>}
  n22 = Normalize(VectorInput=Multiply(A=ObjectRadius(), B=Transform(None=@n21) {transformSourceType=TRANSFORMSOURCE_Local}))
  n23 = ScalarParameter() {defaultValue=0.3, parameterName=Spinner_Speed}
  n24 = LinearInterpolate(Alpha=Sine(None=MaterialFunctionCall(Speed=@n23) {materialFunction=&TimeWithSpeedVariable}.Time) {period=2}) {constB=0.05}
  n25 = MaterialFunctionCall() {materialFunction=&ObjectPivotPoint}
  n26 = WorldPosition()
  n27 = Add(
   A=Transform(None=Multiply(A=@n19.Result, B=@n20)) {transformSourceType=TRANSFORMSOURCE_World, transformType=TRANSFORM_Local}
   B=Multiply(
    A=VertexColor().G
    B=RotateAboutAxis(NormalizedRotationAxis=@n22, RotationAngle=@n24, PivotPoint=@n25.Object Pivot Location, Position=@n26.XYZ)
   )
  )
  NamedRerouteDeclaration(None=@n27) {name=WPO, nodeColor=<rgba 1 0 0.8672 1>}
  n28 = Multiply(A=@n4.RGB, B=@n5)
  NamedRerouteDeclaration(None=@n28) {name=RGB Skulls, nodeColor=<rgba 1 0 0.4687 1>}
  n29 = Constant() {r=1.5}
  n30 = TextureCoordinate() {uTiling=4, vTiling=4}
  n31 = Add(A=@n30, B=Constant2Vector() {g=-0.5, r=-2.45})
  MaterialExpressionConstantClamp_6 = Saturate(None=OneMinus(None=DotProduct(A=@n31, B=@n31)))
  MaterialExpressionConstantClamp_4 = Saturate(None=ConstantBiasScale(None=@MaterialExpressionConstantClamp_6) {bias=-0.95, scale=10})
  n32 = Add(A=@n30, B=Constant2Vector() {g=-0.5, r=-3.45})
  MaterialExpressionConstantClamp_7 = Saturate(None=OneMinus(None=DotProduct(A=@n32, B=@n32)))
  MaterialExpressionConstantClamp_5 = Saturate(None=ConstantBiasScale(None=@MaterialExpressionConstantClamp_7) {bias=-0.95, scale=10})
  MaterialExpressionConstantClamp_8 = Saturate(None=ConstantBiasScale(None=@MaterialExpressionConstantClamp_6) {bias=-0.98, scale=16})
  MaterialExpressionConstantClamp_9 = Saturate(None=ConstantBiasScale(None=@MaterialExpressionConstantClamp_7) {bias=-0.98, scale=16})
  MaterialExpressionConstantClamp_3 = Saturate(None=Add(
    A=Add(A=@MaterialExpressionConstantClamp_4, B=@MaterialExpressionConstantClamp_5)
    B=Add(A=@MaterialExpressionConstantClamp_8, B=@MaterialExpressionConstantClamp_9)
   ))
  MaterialExpressionConstantClamp_2 = Saturate(None=OneMinus(None=Multiply(A=@n29, B=@MaterialExpressionConstantClamp_3)))
  NamedRerouteDeclaration(None=@MaterialExpressionConstantClamp_2) {name=Mask eyes, nodeColor=<rgba 0.7812 1 0 1>}
  n33 = Fresnel(Normal=Transform(None=TextureSample() {samplerType=SAMPLERTYPE_Normal, texture=/DealershipSharedLibrary/FX/Textures/Normal/T_VCVFX_Hat_Skullfire_N}.RGB)) {exponent=3}
  n34 = OneMinus(None=@n33)
  n35 = ScalarParameter() {defaultValue=20, parameterName=FresnelIn}
  n36 = ScalarParameter() {defaultValue=8, parameterName=FresnelOut}
  n37 = NamedRerouteDeclaration(None=Add(A=Power(Base=@n33, Exp=@n36), B=Power(Base=@n34, Exp=@n35))) {name=Fresnel OP, nodeColor=<rgba 0 0.625 1 1>}
  n38 = ScalarParameter() {defaultValue=0.5, parameterName=FresnelMul}
  NamedRerouteDeclaration(None=Add(A=@n6, B=Multiply(A=@n37, B=@n38))) {name=Normal, nodeColor=<rgba 0 0.2593 0.7305 1>}
  n39 = Multiply(A=@n5, B=@n4.A)
  NamedRerouteDeclaration(None=@n39) {name=Particle Color A Skulls, nodeColor=<rgba 0 0.7891 1 1>}
  NamedRerouteUsage()
  Reroute(None=@n39)
  Reroute(None=@n28)
  Reroute(None=@n27)
  StaticSwitchParameter(True=@n1, False=@n10) {parameterName=IsRGBSkulls?}
  StaticSwitchParameter(True=Multiply(A=@n2.Result, B=NamedRerouteUsage()), False=Multiply(A=NamedRerouteUsage(), B=ParticleColor().A)) {parameterName=IsRGBSkulls?}
 ]
}
