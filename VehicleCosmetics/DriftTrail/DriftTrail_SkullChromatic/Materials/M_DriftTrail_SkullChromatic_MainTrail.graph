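# Translucent unlit Niagara ribbon and mesh-particle material for the main trail of the SkullChromatic drift trail effect.
# Emissive tints panning trail textures, warped by smoke and sine offsets, with particle color.
# Opacity builds from layered distorted noise, particle alpha, a dynamic-parameter power and depth fade, while sine waves form a world position offset.
# An IsWheelWrap scalar acts as a switch, altering brightness falloff, wave axes and wheel-wrap opacity masking.
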Material /VehicleCosmetics/DriftTrail/DriftTrail_SkullChromatic/Materials/M_DriftTrail_SkullChromatic_MainTrail
graph {
 properties {
  bUsedWithNiagaraMeshParticles=true
  bUsedWithNiagaraRibbons=true
  blendMode=BLEND_Translucent
  shadingModel=MSM_Unlit
  usedShadingModels=Unlit
 }
 nodes [
  n1 = Constant() {r=1.5}
  n2 = MaterialFunctionCall(
   UVs=TextureCoordinate() {vTiling=0.8}
   Texture Scale=AppendVector(A=ScalarParameter() {defaultValue=1, parameterName=ScaleUvsX}, B=ScalarParameter() {defaultValue=1, parameterName=ScaleUvsY})
  ) {materialFunction=&ScaleUVsByCenter}
  n3 = ScalarParameter() {defaultValue=1, parameterName=ScaleUvsX}
  n4 = ScalarParameter() {defaultValue=1, parameterName=ScaleUvsY}
  n5 = TextureSample(UVs=MaterialFunctionCall(UVs=TextureCoordinate() {uTiling=0.5, vTiling=0.5}, Texture Scale=AppendVector(A=@n3, B=@n4)) {materialFunction=&ScaleUVsByCenter}.UVs) {texture=/DealershipSharedLibrary/FX/Textures/TexturePackage/T_VCVFX_SkullChromatic_Smoke_RGB}
  n6 = Constant() {r=2}
  n7 = Constant() {r=0.1}
  n8 = TextureCoordinate()
  n9 = ComponentMask(None=@n8) {r=true}
  n10 = ScalarParameter() {defaultValue=0.5, parameterName=sine}
  n11 = Sine(None=Add(
    A=MaterialFunctionCall(Speed=@n7) {materialFunction=&TimeWithSpeedVariable}.Time
    B=Multiply(A=@n9, B=@n10)
   ))
  n12 = ComponentMask(None=@n8) {g=true}
  n13 = AppendVector(
   A=NamedRerouteDeclaration(None=Divide(A=@n5.G, B=@n6)) {name=Smoke texture mask, nodeColor=<rgba 0 1 0.25 1>}
   B=Multiply(A=Add(A=@n11, B=@n12), B=Constant() {r=0.075})
  )
  n14 = ScalarParameter() {defaultValue=1, parameterName=TimeTrail}
  n15 = TextureSample(UVs=Panner(
    Coordinate=Add(A=@n2.UVs, B=@n13)
    Time=MaterialFunctionCall(Speed=@n14) {materialFunction=&TimeWithSpeedVariable}.Time
   ) {speedX=-1.25}) {samplerSource=SSM_Wrap_WorldGroupSettings, texture=/DealershipSharedLibrary/FX/Textures/Gradient_Masks/T_VCVFX_SkullChromatic_Trail_A}
  n16 = Add(A=Multiply(A=@n1, B=NamedRerouteUsage()), B=Multiply(A=@n1, B=@n15.RGB))
  n17 = TextureCoordinate()
  n18 = Floor(None=ScalarParameter() {parameterName="IsWheelWrap? 0=False 1=true"})
  n19 = LinearInterpolate(A=OneMinus(None=ComponentMask(None=@n17) {r=true}), Alpha=Saturate(None=Abs(None=@n18))) {constA=1}
  MaterialExpressionConstantClamp_0 = Saturate(None=Multiply(A=ParticleColor().RGB, B=Multiply(A=@n16, B=@n19)))
  MaterialFunctionCall(Emissive=@MaterialExpressionConstantClamp_0) {materialFunction=&MF_FX_EmissiveAdjustments}
  n20 = ComponentMask(None=TextureCoordinate()) {r=true}
  MaterialExpressionConstantClamp_9 = Saturate(None=Multiply(A=@n20, B=Constant() {r=9.5}))
  n21 = Constant() {r=2.5}
  MaterialExpressionConstantClamp_10 = Saturate(None=Multiply(A=OneMinus(None=@n20), B=Constant() {r=2}))
  n22 = Constant() {r=2.5}
  NamedRerouteDeclaration(None=Multiply(A=Power(Base=@MaterialExpressionConstantClamp_9, Exp=@n21), B=Power(Base=@MaterialExpressionConstantClamp_10, Exp=@n22))) {name=Mask One, nodeColor=<rgba 0.5 0 1 1>}
  n23 = ComponentMask(None=TextureCoordinate() {coordinateIndex=1}) {r=true}
  n24 = ScalarParameter() {defaultValue=0.5, parameterName=CountWaves2}
  n25 = ScalarParameter() {defaultValue=7, parameterName=Speed}
  n26 = OneMinus(None=Time())
  n27 = Sine(None=Add(A=Divide(A=@n23, B=@n24), B=Multiply(A=@n25, B=@n26)))
  n28 = ScalarParameter() {defaultValue=5, parameterName=Strenght}
  n29 = Power(Base=ComponentMask(None=TextureCoordinate()) {r=true}, Exp=ScalarParameter() {defaultValue=0.3, parameterName=PowerMiddle2})
  n30 = ScalarParameter() {defaultValue=0.5, parameterName=MultPower2}
  n31 = Multiply(A=OneMinus(None=NamedRerouteUsage()), B=Multiply(A=Multiply(A=@n27, B=@n28), B=Multiply(A=@n29, B=@n30)))
  n32 = MaterialFunctionCall(Y=@n31, Z=@n31) {materialFunction=&MakeFloat3}
  n33 = MaterialFunctionCall(Y=@n31) {materialFunction=&MakeFloat3}
  n34 = Abs(None=Floor(None=ScalarParameter() {parameterName="IsWheelWrap? 0=False 1=true"}))
  NamedRerouteDeclaration(None=LinearInterpolate(
    A=Transform(None=@n32.Result) {transformSourceType=TRANSFORMSOURCE_Local}
    B=Transform(None=@n33.Result) {transformSourceType=TRANSFORMSOURCE_Local}
    Alpha=Saturate(None=@n34)
   ) {constA=1}) {name=WPO, nodeColor=<rgba 0.275 0 1 1>}
  n35 = TextureCoordinate() {uTiling=2}
  n36 = AppendVector(A=ScalarParameter() {defaultValue=1, parameterName=ScaleUvsX}, B=ScalarParameter() {defaultValue=1, parameterName=ScaleUvsY})
  n37 = DynamicParameter() {paramNames [PowerTex, TimeMainTrail, Mirror, Param4]}
  n38 = Constant()
  n39 = MaterialFunctionCall(Speed=Constant() {r=0.1}) {materialFunction=&TimeWithSpeedVariable}
  n40 = TextureCoordinate() {vTiling=0.25}
  n41 = Multiply(A=ComponentMask(None=@n40) {r=true}, B=Constant() {r=1.5})
  n42 = Add(A=Sine(None=Add(A=@n39.Frac Time, B=@n41)), B=ComponentMask(None=@n40) {g=true})
  n43 = Constant() {r=0.02}
  n44 = Add(
   A=Add(
    A=MaterialFunctionCall(UVs=@n35, Texture Scale=@n36) {materialFunction=&ScaleUVsByCenter}.UVs
    B=AppendVector(A=@n37.Mirror, B=@n38)
   )
   B=AppendVector(A=NamedRerouteUsage(), B=Multiply(A=@n42, B=@n43))
  )
  n45 = MaterialFunctionCall(Speed=ScalarParameter() {defaultValue=1, parameterName=TimeTrail}) {materialFunction=&TimeWithSpeedVariable}
  NamedRerouteDeclaration(None=TextureSample(UVs=Panner(Coordinate=@n44, Time=@n45.Time) {speedX=-0.85}) {texture=&T_VCVFX_SkullChromatic_Trail_A}.RGB) {name=RGB Secondary Texture, nodeColor=<rgba 0.5703 0 1 1>}
  MaterialExpressionVertexColor_0 = ParticleColor()
  n46 = TextureCoordinate() {uTiling=1.5, vTiling=1.5}
  n47 = AppendVector(A=ScalarParameter() {defaultValue=1, parameterName=ScaleUvsX}, B=ScalarParameter() {defaultValue=1, parameterName=ScaleUvsY})
  n48 = ScalarParameter() {defaultValue=1, parameterName=ScaleUvsX}
  n49 = ScalarParameter() {defaultValue=1, parameterName=ScaleUvsY}
  n50 = TextureCoordinate()
  n51 = Frac(None=Rotator(Coordinate=@n50, Time=Constant() {r=0.2}))
  n52 = Add(A=ComponentMask(None=@n51) {g=true}, B=OneMinus(None=Time()))
  MaterialExpressionConstantClamp_5 = Saturate(None=ConstantBiasScale(None=Sine(None=@n52) {period=0.5}) {bias=-0.05})
  n53 = Add(
   A=ComponentMask(None=@n51) {g=true}
   B=MaterialFunctionCall(Speed=Constant() {r=-0.5}) {materialFunction=&TimeWithSpeedVariable}.Time
  )
  MaterialExpressionConstantClamp_6 = Saturate(None=ConstantBiasScale(None=Sine(None=@n53) {period=0.25}) {bias=-0.075})
  n54 = Add(A=@MaterialExpressionConstantClamp_5, B=@MaterialExpressionConstantClamp_6)
  n55 = Frac(None=Rotator(Coordinate=@n50, Time=Constant() {r=0.45}) {speed=3.14})
  n56 = OneMinus(None=Time())
  n57 = Add(A=ComponentMask(None=@n55) {r=true}, B=@n56)
  MaterialExpressionConstantClamp_7 = Saturate(None=ConstantBiasScale(None=Sine(None=@n57) {period=0.5}) {bias=-0.05})
  n58 = Add(A=ComponentMask(None=@n55) {r=true}, B=Multiply(A=@n56, B=Constant() {r=0.25}))
  MaterialExpressionConstantClamp_8 = Saturate(None=ConstantBiasScale(None=Sine(None=@n58) {period=0.25}) {bias=-0.085})
  n59 = Add(A=@MaterialExpressionConstantClamp_7, B=@MaterialExpressionConstantClamp_8)
  n60 = Add(
   A=MaterialFunctionCall(UVs=TextureCoordinate() {vTiling=0.5}, Texture Scale=AppendVector(A=@n48, B=@n49)) {materialFunction=&ScaleUVsByCenter}.UVs
   B=Multiply(A=Constant() {r=0.1}, B=AppendVector(A=@n54, B=@n59))
  )
  n61 = OneMinus(None=TextureSample(UVs=Panner(Coordinate=@n60) {speedX=-1.25}) {texture=/DealershipSharedLibrary/FX/Textures/TexturePackage/T_VCVFX_SkullNoise_RGB}.R)
  n62 = Constant() {r=3}
  n63 = Panner(
   Coordinate=Add(
    A=MaterialFunctionCall(UVs=@n46, Texture Scale=@n47) {materialFunction=&ScaleUVsByCenter}.UVs
    B=Divide(A=@n61, B=@n62)
   )
   Time=Multiply(A=Time(), B=ScalarParameter() {defaultValue=1, parameterName=Time})
  ) {speedX=-1}
  MaterialExpressionConstantClamp_4 = Saturate(None=Multiply(A=Constant() {r=2}, B=TextureSample(UVs=@n63) {texture=/DealershipSharedLibrary/FX/Textures/Noise/T_VCVFX_NoiseText_A}.RGB))
  n64 = Multiply(A=Multiply(A=@MaterialExpressionConstantClamp_4, B=NamedRerouteUsage()), B=NamedRerouteUsage())
  n65 = DynamicParameter() {paramNames [PowerTex, TimeMainTrail, Mirror, Param4]}
  MaterialExpressionConstantClamp_3 = Saturate(None=Multiply(A=@MaterialExpressionVertexColor_0.A, B=Power(Base=@n64, Exp=@n65.PowerTex)))
  n66 = ScalarParameter() {defaultValue=30, parameterName=DepthFade Distance}
  n67 = MaterialFunctionCall!(Multiply, Power EXP) {materialFunction=/DealershipSharedLibrary/FX/Materials/MaterialFunctions/MF_VCVFX_WheelwrapMask}
  n68 = ScalarParameter() {parameterName="IsWheelWrap? 0=False 1=true"}
  n69 = Saturate(None=Abs(None=Floor(None=@n68)))
  NamedRerouteDeclaration(None=Multiply(
    A=MaterialFunctionCall(Opacity In=@MaterialExpressionConstantClamp_3, DepthFade Distance=@n66) {materialFunction=&MF_VCVFX_DepthFade}.Result
    B=LinearInterpolate(B=@n67.Result, Alpha=@n69) {constA=1}
   )) {name=Opacity, nodeColor=<rgba 0.1484 1 0 1>}
  NamedRerouteUsage()
  NamedRerouteUsage()
 ]
}
